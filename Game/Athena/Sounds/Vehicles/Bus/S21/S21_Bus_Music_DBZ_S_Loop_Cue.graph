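# Sound cue for looping Season 21 Battle Bus music, a DBZ variant by its name.
# A single looping wave player feeds the root directly, with a boosted cue volume.
# It routes through the Battle Bus sound class and references the shared bus music attenuation asset.
# Its inline attenuation block turns off attenuation and spatialization, but the override switch is left at default, so the shared asset applies.
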SoundCue /Game/Athena/Sounds/Vehicles/Bus/S21/S21_Bus_Music_DBZ_S_Loop_Cue
root=WavePlayer_0
nodes [{id=WavePlayer_0, properties {bLooping=true, soundWaveAssetPtr=&/s21_Bus_Music_DBZ_S}}]
properties {
 attenuationSettings=/Game/Sounds/Attenuation/Fort_Bus_Music_Athena_Att
 maxDistance=21000
 soundClassObject=/Game/Sounds/FortSoundClassesAndModes/Fort_BattleBus
 volumeMultiplier=1.65
 attenuationOverrides {bAttenuate=false, bSpatialize=false}
}
