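# Sound cue that loops the Season 21 "Vibin" music track for the Battle Bus.
# A single looping wave player is the root, with no randomization, mixing or other processing nodes.
# The cue routes to the Battle Bus sound class with a boosted volume and references the Athena bus music attenuation asset for its distance behavior.
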SoundCue /Game/Athena/Sounds/Vehicles/Bus/S21/S21_Bus_Music_Vibin_Loop_Cue
root=WavePlayer_10
nodes [{id=WavePlayer_10, properties {bLooping=true, soundWaveAssetPtr=&/S21_Bus_Music_Vibin}}]
properties {
 attenuationSettings=/Game/Sounds/Attenuation/Fort_Bus_Music_Athena_Att
 maxDistance=21000
 soundClassObject=/Game/Sounds/FortSoundClassesAndModes/Fort_BattleBus
 volumeMultiplier=1.397
 attenuationOverrides {bAttenuate=false, bSpatialize=false}
}
